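# Green gem material for the AxisMandate emerald container prop, Nanite-enabled, that fakes faceted crystal reflections and refraction.
# Cubemaps, sparkles, sun-direction highlights, Fresnel, fake Phong and time-of-day terms build the color layers, linked through named reroutes.
# Base color and emissive use inverse eye adaptation to offset exposure, while refraction varies with normal direction and pixel depth.
# The assembled material attributes pass through the shared MF_WPOAnim function.
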Material /Hera/Props/AxisMandate/AxisMandate_EmeraldContainer_A/Emerald/Material/M_NarrowFlea_Gem_C
graph {
 properties {
  bAutomaticallySetUsageInEditor=false
  bUseMaterialAttributes=true
  bUsedWithNanite=true
  previewMesh=/NarrowFleaJade/Assets/Meshes/NarrowFlea_Gem_DS
  translucencyLightingMode=TLM_VolumetricDirectional
  usedShadingModels=Default Lit
 }
 nodes [
  n1 = ScalarParameter() {defaultValue=0.3781, parameterName=Fresnel Size}
  n2 = NamedRerouteUsage()
  n3 = Saturate(None=Power(
    Base=Fresnel(ExponentIn=@n1, Normal=@n2) {baseReflectFraction=0}
    Exp=ScalarParameter() {defaultValue=3.6938, parameterName=Fresnel Contrast}
   ))
  n4 = ScalarParameter() {parameterName=Fresnel facing down mask}
  n5 = NamedRerouteUsage()
  n6 = Multiply(
   A=@n3
   B=Multiply(A=ScalarParameter() {defaultValue=0.272, parameterName=Fresnel Intensity}, B=LinearInterpolate(B=@n4, Alpha=@n5) {constA=1})
  )
  n7 = SkyAtmosphereLightDirection()
  n8 = Saturate(None=Power(Base=DotProduct(A=@n2, B=@n7), Exp=ScalarParameter() {defaultValue=7, parameterName=Fresnel Light dir pw}))
  n9 = ScalarParameter() {defaultValue=1, parameterName=Fresnel Light dir mult}
  Add(A=@n6, B=Multiply(A=Multiply(A=@n8, B=@n9), B=NamedRerouteUsage()))
  MaterialFunctionCall(Hue Shift Percentage=ScalarParameter() {defaultValue=3.6827, parameterName=Para56}) {materialFunction=&HueShift}
  MaterialFunctionCall!(Sine Phase=Multiply(A=Time()) {constB=0.2}, Value 1=Constant() {r=0.5}, Value 2=Constant() {r=1.5}) {materialFunction=/Engine/Functions/Engine_MaterialFunctions02/Utility/Sine_Remapped}
  n10 = Add(A=NamedRerouteUsage(), B=ScalarParameter() {defaultValue=-0.1234, parameterName=Cubemap repetition})
  n11 = TextureSampleParameterCube!(UVs=@n10, Apply View MipBias) {group=Cube Map, parameterName=ReflectionCubeMap2, texture=/Game/CharClasses/Technician/Gadgets/BASE/Effects/T_CrystalColor_Tile}
  n12 = Multiply(
   A=LinearInterpolate(A=NamedRerouteUsage()) {constAlpha=0.6}
   B=LinearInterpolate(A=@n11.RGB, Alpha=ScalarParameter() {parameterName=Cubemap 2 detail})
  )
  Multiply(A=@n12) {constB=0.75}
  NamedRerouteDeclaration(None=TextureSample(UVs=NamedRerouteUsage()) {samplerType=SAMPLERTYPE_Normal, texture=&/T_NarrowFlea_Jade_N}.RGB) {name=HP normals, nodeColor=<rgba 0 0.1684 1 1>}
  NamedRerouteDeclaration(None=@n3) {name=Fresnel Base, nodeColor=<rgba 1 0 0.4453 1>}
  n13 = NamedRerouteUsage()
  n14 = ReflectionVectorWS(CustomWorldNormal=NamedRerouteDeclaration(None=Transform(None=@n13)) {name=World Normals, nodeColor=<rgba 0 0.0391 1 1>})
  n15 = PixelNormalWS()
  n16 = DotProduct(A=@n14, B=@n15)
  n17 = ScalarParameter() {defaultValue=1, group=mbient Light, parameterName=Phong Normal Mult}
  n18 = Subtract(
   A=Multiply(A=Multiply(A=@n16, B=@n17) {constB=2}, B=@n15)
   B=NamedRerouteDeclaration(None=@n14) {name=Final Reflection Vector, nodeColor=<rgba 0.3594 1 0 1>}
  )
  n19 = MaterialFunctionCall() {materialFunction=&CameraDirectionVector}
  n20 = Max(A=DotProduct(A=Normalize(VectorInput=@n18), B=Multiply(A=@n19.Result) {constB=-1})) {constB=0}
  n21 = ScalarParameter() {defaultValue=12.166, group=mbient Light, parameterName=Fake Phong Contrast}
  n22 = Clamp(
   None=Multiply(A=Power(Base=@n20, Exp=@n21), B=ScalarParameter() {defaultValue=1.0487, group=mbient Light, parameterName=Fake Phong Intensity})
   Max=ScalarParameter() {defaultValue=1, parameterName=Phong max}
  ) {maxDefault=0.5}
  n23 = ComponentMask(None=Multiply(A=VertexNormalWS()) {constB=-1}) {b=true}
  n24 = LinearInterpolate(
   B=ScalarParameter() {defaultValue=1, parameterName=Phong Down Face Mask}
   Alpha=NamedRerouteDeclaration(None=Saturate(None=@n23)) {name=Down face mask, nodeColor=<rgba 0.0312 0 1 1>}
  ) {constA=1}
  NamedRerouteDeclaration(None=MaterialFunctionCall!(Default=Multiply(A=@n22, B=@n24), Low/Mobile=Constant()) {materialFunction=/Game/Effects/Fort_Effects/Materials/_Functions/MF_LowMobile_PathingSwitch}.Result) {name=Phong Reflections, nodeColor=<rgba 1 0 0.6562 1>}
  NamedRerouteDeclaration(None=ScalarParameter() {defaultValue=1, parameterName=Metallic}) {name=Final Metallic, nodeColor=<rgba 1 0.3047 0 1>}
  NamedRerouteDeclaration(None=ScalarParameter() {defaultValue=0.1, parameterName=Rough}) {name=Final Roughness, nodeColor=<rgba 0 0.0156 1 1>}
  n25 = VertexNormalWS()
  n26 = LinearInterpolate(
   A=ScalarParameter() {defaultValue=0.9, parameterName=Refraction Min}
   B=ScalarParameter() {defaultValue=0.446, parameterName=Refraction Max}
   Alpha=Saturate(None=ComponentMask(None=@n25) {b=true})
  ) {constA=0.81}
  n27 = PixelDepth()
  n28 = ScalarParameter() {parameterName=Refraction Fade Distance}
  n29 = Saturate(None=Divide(A=Add(A=@n27.R, B=@n28)) {constB=600})
  n30 = NamedRerouteUsage()
  n31 = TextureSampleParameterCube(UVs=@n10) {
   group=Cube Map
   parameterName=ReflectionCubeMap
   samplerType=SAMPLERTYPE_LinearColor
   texture=/Game/Weapons/WeaponSkins/Textures/AthenaCubemap2
  }
  n32 = Multiply(A=NamedRerouteUsage(), B=@n31.RGB)
  n33 = NamedRerouteUsage()
  n34 = Multiply(A=@n33) {constB=-1}
  n35 = ScalarParameter() {defaultValue=1, parameterName=Param_2}
  n36 = Multiply(A=ScalarParameter() {defaultValue=1, group=Sparkles, parameterName=Sparkle Intensity}, B=NamedRerouteUsage())
  n37 = Multiply(A=NamedRerouteUsage(), B=@n36)
  n38 = Add(
   A=LinearInterpolate(
    A=@n36
    B=LinearInterpolate(A=@n31.RGB, B=@n37)
    Alpha=ScalarParameter() {defaultValue=0.3, group=Cube Map, parameterName=Cube Map Intensity}
   ) {constAlpha=0.7}
   B=Multiply(A=NamedRerouteUsage(), B=NamedRerouteUsage())
  )
  n39 = NamedRerouteUsage()
  n40 = SkyAtmosphereLightDirection()
  n41 = Multiply(
   A=Power(Base=DotProduct(A=@n39, B=@n40), Exp=ScalarParameter() {defaultValue=2.1721, parameterName=Reflection Sun pw})
   B=ScalarParameter() {defaultValue=2, parameterName=Reflection Sun Mult}
  )
  n42 = Power(
   Base=DotProduct(A=VertexNormalWS(), B=NamedRerouteUsage())
   Exp=ScalarParameter() {defaultValue=131.7417, parameterName=Vtx Normal Compare PW}
  )
  n43 = ScalarParameter() {parameterName=Vtx Normal Compare Mult}
  n44 = Multiply(A=Add(A=Add(A=@n38, B=@n41), B=Multiply(A=@n42, B=@n43)), B=NamedRerouteUsage())
  n45 = ScalarParameter() {defaultValue=1, parameterName=Emissive overall intensity}
  n46 = MakeMaterialAttributes!(
   BaseColor=EyeAdaptationInverse(LightValueInput=Multiply(A=@n30, B=@n32), AlphaInput=ScalarParameter() {defaultValue=0.7, parameterName=Base Color Eye adap}).EyeAdaptationInverse
   Metallic=NamedRerouteUsage()
   Specular
   Roughness=NamedRerouteUsage()
   Anisotropy
   EmissiveColor=EyeAdaptationInverse(LightValueInput=Multiply(A=@n44, B=@n45), AlphaInput=ScalarParameter() {defaultValue=0.5, parameterName=Emissive Eye Adapt}).EyeAdaptationInverse
   Opacity=ScalarParameter() {defaultValue=0.75, parameterName=opacity}
   OpacityMask
   Normal=NamedRerouteDeclaration(None=LinearInterpolate(A=@n34, B=@n33, Alpha=@n35)) {name=Normals, nodeColor=<rgba 0 0.1331 1 1>}
   Tangent
   WorldPositionOffset
   SubsurfaceColor
   ClearCoat
   ClearCoatRoughness
   AmbientOcclusion
   Refraction=Saturate(None=LinearInterpolate(A=@n26, Alpha=@n29) {constA=1})
   CustomizedUVs_0
   CustomizedUVs_1
   CustomizedUVs_2
   CustomizedUVs_3
   CustomizedUVs_4
   CustomizedUVs_5
   CustomizedUVs_6
   CustomizedUVs_7
   PixelDepthOffset
   ShadingModel
   Displacement
  )
  NamedRerouteDeclaration(None=GetMaterialAttributes(None=MaterialFunctionCall!(In=@n46) {materialFunction=/Game/Athena/WinterArt/Materials/MF_WPOAnim}.Result) {attributeGetTypes [#1]}.Normal) {name=Normal Rotated, nodeColor=<rgba 1 0 0.5625 1>}
  NamedRerouteDeclaration(None=@n6) {name=Fresnel, nodeColor=<rgba 0 0.2734 1 1>}
  n47 = Multiply(A=@n12, B=ScalarParameter() {defaultValue=1, group=Gemstone, parameterName=GS_ReflectionsBrightness})
  n48 = Desaturation(
   None=Power(Base=Saturate(None=@n47), Exp=ScalarParameter() {defaultValue=3, group=Gemstone, parameterName=GS_RefractionsExponent})
   Fraction=ScalarParameter() {group=Gemstone, parameterName=GS_RefractionsDesaturation}
  )
  NamedRerouteDeclaration(None=@n48) {name=Sky Refractions, nodeColor=<rgba 0.6875 1 0 1>}
  NamedRerouteDeclaration(None=ReflectionVectorWS(CustomWorldNormal=NamedRerouteUsage())) {name=ReflectionVector, nodeColor=<rgba 1 0 0.3281 1>}
  n49 = CollectionParameter() {
   collection=&FortniteMaterialParameters
   parameterName=FortniteActiveTimeOfDay
  }
  n50 = MaterialFunctionCall(Float4=@n49) {materialFunction=&BreakOutFloat4Components}
  NamedRerouteDeclaration(None=@n50.G) {name=Day, nodeColor=<rgba 0.9219 1 0 1>}
  n51 = DotProduct(A=VertexNormalWS(), B=CameraVectorWS())
  NamedRerouteDeclaration(None=Saturate(None=Power(Base=@n51) {constExponent=5})) {name=Gradient, nodeColor=<rgba 0.6172 1 0 1>}
  NamedRerouteDeclaration(None=Multiply(A=TextureCoordinate(), B=ScalarParameter() {defaultValue=1, group=PBR Textures, parameterName=UV Tiling})) {name=UV Size, nodeColor=<rgba 0.5469 0 1 1>}
  n52 = VectorParameter() {parameterName=Base Color, defaultValue=<rgba 0.0299 0.5 0.0208 1>}
  n53 = NamedRerouteUsage()
  n54 = NamedRerouteDeclaration(None=@n50.R) {name=Morning, nodeColor=<rgba 0.9219 1 0 1>}
  n55 = NamedRerouteDeclaration(None=@n50.B) {name=Aftern, nodeColor=<rgba 0.9219 1 0 1>}
  n56 = Saturate(None=Add(A=Add(A=@n54, B=@n55), B=NamedRerouteDeclaration(None=@n50.A) {name=Night, nodeColor=<rgba 0.9219 1 0 1>}))
  NamedRerouteDeclaration(None=Multiply(A=Add(A=@n52.RGB, B=@n53), B=LinearInterpolate(Alpha=@n56) {constA=1, constB=2})) {name=Diffuse Texture, nodeColor=<rgba 0 0.4141 1 1>}
  NamedRerouteDeclaration(None=LinearInterpolate(A=MaterialFunctionCall() {materialFunction=&Color_Time_of_Day}.Result)) {name=TOD colors, nodeColor=<rgba 0 1 0.0625 1>}
  NamedRerouteUsage()
  Normalize(VectorInput=Subtract(A=WorldPosition().XYZ, B=ObjectPositionWS()))
  ReflectionVectorWS()
  Saturate(None=@n48)
  VertexNormalWS()
 ]
}
